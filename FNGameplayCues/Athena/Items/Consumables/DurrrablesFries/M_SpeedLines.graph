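# Translucent unlit, two-sided speed-line effect material for the DurrrablesFries consumable gameplay cue, usable on mesh particles and Niagara ribbons.
# Particle color feeds an emissive adjustment function for the glow.
# A stretched, panning caustics texture gains a floored boost for hot streaks, then combines with two Fresnel terms to shape the streak mask.
# Particle alpha, vertex red, a dynamic Opacity parameter and depth fade scale that mask.
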Material /FNGameplayCues/Athena/Items/Consumables/DurrrablesFries/M_SpeedLines
graph {
 properties {
  bUsedWithMeshParticles=true
  bUsedWithNiagaraRibbons=true
  blendMode=BLEND_Translucent
  pixelDepthOffsetMode=PDOM_Legacy
  shadingModel=MSM_Unlit
  twoSided=true
  usedShadingModels=Unlit
 }
 nodes [
  n1 = ParticleColor()
  MaterialFunctionCall(Emissive=@n1.RGB) {materialFunction=&MF_FX_EmissiveAdjustments}
  n2 = TextureSample(UVs=Panner(Coordinate=TextureCoordinate() {uTiling=0.18, vTiling=3}) {speedX=-1.25, speedY=0.3}) {texture=&T_ShieldCaustics}
  n3 = Add(A=@n2.RGB) {constB=0.12}
  n4 = Add(A=@n2.RGB, B=Multiply(A=Floor(None=@n3)) {constB=0.5})
  n5 = MaterialFunctionCall(Exponent=Constant() {r=4}) {materialFunction=&MF_Fresnel_FX}
  n6 = Multiply(A=Multiply(A=@n1.A, B=Add(A=@n4, B=@n5.Fresnel Inverted)), B=VertexColor().R)
  n7 = MaterialFunctionCall(Brightness=Constant() {r=1.8}, Exponent=Constant() {r=16}) {materialFunction=&MF_Fresnel_FX}
  Multiply(
   A=Multiply(A=Multiply(A=@n6, B=@n7.Fresnel), B=DynamicParameter() {paramNames [Opacity,Param2,Param3,Param4]}.Opacity)
   B=DepthFade() {fadeDistanceDefault=60}
  )
  n8 = Multiply(A=ComponentMask(None=TextureCoordinate()) {r=true}) {constB=2}
  Multiply(B=Saturate(None=Subtract(A=@n8) {constB=0.3}))
 ]
}
